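# Translucent lit material for the glowing Slurp liquid surface in the Slurp factory, with animated two-color swirl and masked edges.
# Layered water normals, using a separate Forward/Mobile path, drive roughness, metallic and color-mask distortion.
# Water simulation adds swirl distortion, boosts foam glow and cuts opacity, which also fades by depth.
# Time-of-day functions scale color and emissive, with eye-adaptation compensation, while a camera-distance sphere mask blends a vertical offset vector.
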Material /Clyde/Athena/Environments/Materials/SlurpfactorySlurp/M_SilyWolf_Surface_Clyde
graph {
 properties {
  blendMode=BLEND_Translucent
  numCustomizedUVs=2
  pixelDepthOffsetMode=PDOM_Legacy
  translucencyLightingMode=TLM_Surface
  usedShadingModels=Default Lit
 }
 nodes [
  n1 = OneMinus(None=DepthFade(FadeDistance=ScalarParameter() {defaultValue=50, parameterName=Color Contribution Fade Distance}))
  n2 = TextureCoordinate() {coordinateIndex=1}
  n3 = TextureSample(UVs=@n2) {samplerType=SAMPLERTYPE_Normal, texture=/Game/Global/Textures/Default/Tileable_Patterns/T_Water2_N}
  n4 = ComponentMask(None=@n3.RGB) {g=true, r=true}
  n5 = ScalarParameter() {defaultValue=0.05, group=Deep Water, parameterName=Noise 1}
  n6 = TextureSample(UVs=Add(A=Multiply(A=@n4, B=@n5), B=@n2)) {samplerType=SAMPLERTYPE_Normal, texture=&T_Water2_N}
  n7 = TextureCoordinate() {coordinateIndex=1}
  n8 = VertexInterpolator(VS=@n7)
  n9 = TextureSample(UVs=VertexInterpolator(VS=@n7).PS) {samplerType=SAMPLERTYPE_Normal, texture=&T_Water2_N}
  n10 = TextureSample(UVs=VertexInterpolator(VS=@n7).PS) {samplerType=SAMPLERTYPE_Normal, texture=&T_Water2_N}
  n11 = MaterialFunctionCall(
   BaseNormal=MaterialFunctionCall(
    Normal=TextureSample(UVs=@n8.PS) {samplerType=SAMPLERTYPE_Normal, texture=/Game/Global/Textures/Default/Tileable_Patterns/T_WaterSoft_N}.RGB
    Flatness=ScalarParameter() {defaultValue=0.9115, parameterName=NormalFlaten_M3}
   ) {materialFunction=&FlattenNormal}.Result
   AdditionalNormal=MaterialFunctionCall(
    Normal=MaterialFunctionCall(BaseNormal=@n9.RGB, AdditionalNormal=@n10.RGB) {materialFunction=&BlendAngleCorrectedNormals}.Result
    Flatness=ScalarParameter() {parameterName=NormalFlatten}
   ) {materialFunction=&FlattenNormal}.Result
  ) {materialFunction=&BlendAngleCorrectedNormals}
  n12 = ShadingPathSwitch(
   Default=MaterialFunctionCall(
    Normal=MaterialFunctionCall(BaseNormal=@n3.RGB, AdditionalNormal=@n6.RGB) {materialFunction=&BlendAngleCorrectedNormals}.Result
    Flatness=ScalarParameter() {parameterName=NormalFlatten}
   ) {materialFunction=&FlattenNormal}.Result
   Forward=@n11.Result
   Mobile=@n11.Result
  )
  n13 = ComponentMask(None=@n12) {b=true}
  n14 = MaterialFunctionCall!(Flowmap Velocity, Previous Foam, Normal, WPO) {materialFunction=/FortWater/Materials/Functions/WaterSimulationTexture}
  n15 = Multiply(A=ComponentMask(None=@n14.Normal) {g=true, r=true}, B=ScalarParameter() {defaultValue=0.1, parameterName=Water Normal Contribution})
  n16 = Add(A=TextureCoordinate() {uTiling=2, vTiling=2}, B=Saturate(None=Add(A=@n13, B=@n15)))
  n17 = Multiply(A=Time(), B=ScalarParameter() {defaultValue=0.125, parameterName=Color Mask Rotating Speed})
  n18 = LinearInterpolate(
   A=VectorParameter() {parameterName=ColorA, defaultValue=<rgba 0.0008 0.1042 0.0874 1>}.RGB
   B=VectorParameter() {parameterName=ColorB, defaultValue=<rgba 0.0076 0.974 0.8173 1>}.RGB
   Alpha=TextureSample(UVs=Rotator(Coordinate=@n16, Time=@n17)) {samplerType=SAMPLERTYPE_Masks, texture=/Game/Global/Textures/Default/Masks_and_Gradients/Swamp_Mask}.B
  )
  n19 = ScalarParameter() {defaultValue=1, group=Time Of day tweaks, parameterName=Morning Emissive Intensity}
  n20 = ScalarParameter() {defaultValue=1, group=Time Of day tweaks, parameterName=Day Emissive Intensity}
  n21 = ScalarParameter() {defaultValue=1, group=Time Of day tweaks, parameterName=Evening Emissive Intensity}
  n22 = ScalarParameter() {defaultValue=1, group=Time Of day tweaks, parameterName=Night Emissive Intensity}
  n23 = MaterialFunctionCall(
   Value_Morning=ScalarParameter() {defaultValue=15, group=Time Of day tweaks, parameterName=Morning Drain Emissive Intensity}
   Value_Day=ScalarParameter() {defaultValue=15, group=Time Of day tweaks, parameterName=Day Drain Emissive Intensity}
   Value_Evening=ScalarParameter() {defaultValue=15, group=Time Of day tweaks, parameterName=Evening Drain Emissive Intensity}
   Value_Night=ScalarParameter() {defaultValue=15, group=Time Of day tweaks, parameterName=Night Drain Emissive Intensity}
  ) {materialFunction=&Scalar_Time_of_Day}
  n24 = TextureSample() {samplerSource=SSM_Wrap_WorldGroupSettings, texture=/Game/Athena/Items/Gameplay/SuperSilkyWolf/FX/T_SlurpPuddleEmissiveMasks}
  n25 = Add(
   A=Multiply(
    A=LinearInterpolate(B=@n18, Alpha=@n1)
    B=MaterialFunctionCall(Value_Morning=@n19, Value_Day=@n20, Value_Evening=@n21, Value_Night=@n22) {materialFunction=&Scalar_Time_of_Day}.Result
   )
   B=Multiply(
    A=Multiply(A=@n18, B=@n23.Result) {constB=10}
    B=StaticComponentMaskParameter(None=@n24.RGB) {defaultR=true, parameterName=EmissiveMaskChannel}
   )
  )
  n26 = Add(A=ComponentMask(None=@n14.WPO) {b=true}, B=ScalarParameter() {defaultValue=0.25, parameterName=WPO water sim Z Add})
  n27 = Add(
   A=Power(Base=@n26, Exp=ScalarParameter() {defaultValue=4, parameterName=WPO water sim Z pow B})
   B=ScalarParameter() {defaultValue=1, parameterName=WPO water sim Z Add B}
  )
  n28 = ScalarParameter() {defaultValue=10, parameterName=WPO Water sim Z Pow A}
  n29 = Clamp(None=LinearInterpolate(Alpha=Power(Base=@n26, Exp=@n28)) {constA=1, constB=0})
  n30 = LinearInterpolate(B=ScalarParameter() {defaultValue=100, parameterName=Water Sim Foam Emissive Intensity}, Alpha=Clamp(None=Subtract(A=@n27, B=@n29))) {constA=1}
  n31 = ScalarParameter() {defaultValue=1, group=Time Of day tweaks, parameterName=Morning Overall Emissive Multi}
  n32 = ScalarParameter() {defaultValue=1, group=Time Of day tweaks, parameterName=Day Overall Emissive Multi}
  n33 = ScalarParameter() {defaultValue=1, group=Time Of day tweaks, parameterName=Evening Overall Emissive Multi}
  n34 = ScalarParameter() {defaultValue=0.2, group=Time Of day tweaks, parameterName=Night Overall Emissive Multi}
  EyeAdaptationInverse(LightValueInput=Multiply(
    A=Multiply(A=@n25, B=@n30)
    B=MaterialFunctionCall(Value_Morning=@n31, Value_Day=@n32, Value_Evening=@n33, Value_Night=@n34) {materialFunction=&Scalar_Time_of_Day}.Result
   ))
  n35 = Saturate(None=@n13)
  LinearInterpolate(
   A=ScalarParameter() {defaultValue=0.01, parameterName=Roughness Max}
   B=ScalarParameter() {defaultValue=0.25, parameterName=Roughness Min}
   Alpha=@n35
  ) {constA=0.1, constB=0.01}
  LinearInterpolate(B=ScalarParameter() {defaultValue=0.5, parameterName=Metallic Max}, Alpha=@n35) {constB=0.5}
  LinearInterpolate(
   A=VectorParameter() {parameterName=Distance, defaultValue=<rgba 0 0 200 1>}.RGB
   B=Constant3Vector() {constant=<rgba 0 0 5 0>}
   Alpha=SphereMask(
    A=WorldPosition().XYZ
    B=CameraPositionWS()
    Radius=ScalarParameter() {defaultValue=40000, parameterName=Falloff Radius}
    Hardness=ScalarParameter() {parameterName=Falloff Hardness}
   )
  ) {constB=0}
  n36 = TextureSample() {
   samplerSource=SSM_Wrap_WorldGroupSettings
   samplerType=SAMPLERTYPE_Masks
   texture=&Swamp_Mask
  }
  n37 = Multiply(A=ScalarParameter() {defaultValue=0.6, parameterName=Alpha}, B=@n36.G)
  n38 = TextureSample() {constMipValue=0, samplerSource=SSM_Wrap_WorldGroupSettings, texture=/Game/Athena/Items/Gameplay/SuperSilkyWolf/FX/T_SlurpPuddleAlphaMasks}
  n39 = StaticComponentMaskParameter(None=@n38.RGB) {defaultR=true, parameterName=Alpha Mask Channel}
  n40 = Distance(
   A=MaterialFunctionCall() {materialFunction=&WorldPositionBehindTranslucency}.Result
   B=WorldPosition().XYZ
  )
  n41 = ScalarParameter() {defaultValue=20, parameterName=Custom Depth Fade}
  Multiply(A=Multiply(A=Multiply(A=@n37, B=@n39), B=@n29), B=Saturate(None=Divide(A=@n40, B=@n41)))
  Multiply(
   A=@n18
   B=MaterialFunctionCall(
    Value_Morning=ScalarParameter() {defaultValue=1, group=Time Of day tweaks, parameterName=Morning Drain Base Color Intensity}
    Value_Day=ScalarParameter() {defaultValue=1, group=Time Of day tweaks, parameterName=Day Drain Base Color Intensity}
    Value_Evening=ScalarParameter() {defaultValue=1, group=Time Of day tweaks, parameterName=Evening Drain Base Color Intensity}
    Value_Night=ScalarParameter() {defaultValue=1, group=Time Of day tweaks, parameterName=Night Drain Base Color Intensity}
   ) {materialFunction=&Scalar_Time_of_Day}.Result
  ) {constB=10}
  Reroute(None=@n12)
  n42 = WorldPosition()
  Rotator(
   Coordinate=Divide(A=ComponentMask(None=@n42.XYZ) {g=true, r=true}, B=ScalarParameter() {defaultValue=5000, parameterName=XTile})
   Time=Multiply(A=Time(), B=ScalarParameter() {defaultValue=-0.001, parameterName=Primary Rotation Speed})
  )
 ]
}
